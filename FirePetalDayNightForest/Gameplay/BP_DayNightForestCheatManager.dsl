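; Cheat manager for the FirePetal Day/Night Forest feature, providing debug shortcuts to test its NPC and vault flow.
; Helpers grant a water sprite or vault keycard, teleport to the Bushranger NPC or vault door, and broadcast a time-of-day change request through the global event router.
; The NPC flow simulation checks authority and the pawn's sprite-in-inventory tag, then gives the keycard and triggers the time-of-day change.
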(fn DayNightForestSimulateNPCFlow
  (bind _returnvalue (CheatManager|GetPlayerController))
  (if (Networking|HasAuthority _returnvalue)
    (bind _asfort_player_controller_athena (Utilities|Casting|CastToFortPlayerControllerAthena _returnvalue))
    (if (GameplayTags|HasMatchingGameplayTag (Game|Player|GetPlayerPawn _asfort_player_controller_athena) "(TagName=\"Athena.Player.Sprite.InInventory\")")
      (CallFunction|GiveDayNightForestVaultKeycard)
      (CallFunction|TriggerDayNightForestTODChange))))

(fn GiveDayNightForestNPCRequirements
  (Development|ExecuteConsoleCommand "Cheat GiveConsumable WID_Sprite_Held_Water" (CheatManager|GetPlayerController)))

(fn GiveDayNightForestVaultKeycard
  (bind _asfort_player_controller_athena (Utilities|Casting|CastToFortPlayerControllerAthena (CheatManager|GetPlayerController)))
  (InventoryManagement|AddItem _asfort_player_controller_athena "/FirePetalDayNightForest/Gameplay/KeyItem/WID_FirePetal_DayNight_VaultKey.WID_FirePetal_DayNight_VaultKey" 1 false))

(fn GoToDayNightForestNPCLocation
  (Development|ExecuteConsoleCommand "Cheat LivingWorldManagerTeleportTo Bushranger" (CheatManager|GetPlayerController)))

(fn GoToDayNightForestVaultDoor
  (Development|ExecuteConsoleCommand "BugItGo -75798.629983 4353.254038 2393.150003 -9.021320 -2.637180 0.000000" (CheatManager|GetPlayerController)))

(fn TriggerDayNightForestTODChange
  (GameplayEventRouter|BroadcastEvent (GameplayEventRouter|GetGlobalEventRouter) (Utilities|Struct|MakeOnTimeOfDayChangeRequested "(GameplayTags=)") 0))
