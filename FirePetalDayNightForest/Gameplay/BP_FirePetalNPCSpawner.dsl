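; BP_FirePetalNPCSpawner is a gameplay Blueprint that randomly thins a set of pre-placed temporary NPC actors at game start.
; On BeginPlay it waits for the Fortnite playlist data to be ready, then shuffles its NPC temp actor array.
; It loops over a fixed three slots and destroys every actor except the one at index 2, leaving a random survivor.
; Despite the name, it spawns nothing itself.
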(event EventBeginPlay
  (Class|FortAsyncActionCurrentPlaylistReady|PlaylistDataReadyAsync
    (:OnReady
      (Utilities|Array|Shuffle (Variables|Default|GetNPCTempActors))
      (for _index (range 3)
        (if (not (== _index 2))
          (Actor|DestroyActor (Utilities|Array|Get(acopy) (Variables|Default|GetNPCTempActors) _index)))))))

(fn ConstructionScript)
