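; Key-locked FirePetal day/night vault door that a player carrying the vault key interacts with to open once.
; A time-of-day change event sets a replicated flag whose OnRep swaps door and frame to dynamic materials.
; Interaction consumes the key if required, then a timeline moves the pivot, drives emissive and triggers client-only sound, particles and camera shake.
; A server-side kill volume kills campers overlapping it before opening, then loses collision.
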(event EventBeginPlay
  (bind _returnvalue (GameplayEventRouter|GetGlobalEventRouter))
  ; Init
  (Variables|Default|SetDynamicDoorStartTransform (Transformation|GetWorldTransform (Variables|Default|GetDoorPivot)))
  (Utilities|IsValid _returnvalue
    (:Is%20Valid
      _)
    (:Is%20Not%20Valid
      (Utilities|FlowControl|Delay 2.0)))
  (GameplayEventRouter|StartListeningToEvent _returnvalue "/Game/Conversation/Services/Effects/OnTimeOfDayChangeRequested.OnTimeOfDayChangeRequested"
    (:OnEventReceived
      (Utilities|FlowControl|DoOnce false
        (:Completed
          (Variables|Default|SetDoorVisualsChangedState true))))))

(event Custom|GiveVaultKey
  ; Debug
  (Inventory|SpawnPickupInWorld "/FirePetalDayNightForest/Gameplay/KeyItem/WID_FirePetal_DayNight_VaultKey.WID_FirePetal_DayNight_VaultKey" 1 (Transformation|GetWorldLocation (Variables|Default|GetArrow))))

(event Custom|ChangeDoorState
  ; Init
  (Networking|SwitchHasAuthority
    (:Authority
      (Utilities|FlowControl|DoOnce false
        (:Completed
          (Variables|Default|SetDoorVisualsChangedState true))))))

(event Interaction|EventBlueprintOnInteract
  (params (InteractingPawn :type Fort%20Pawn%20Object%20Reference) (InteractionBeingAttempted :type EInteractionBeingAttempted%20Enum :default "FirstInteraction"))
  ; Interact And Open
  (Variables|Default|SetInteractingPlayer InteractingPawn)
  (if (Variables|Default|GetRequiresKey)
    (bind _haskey (CallFunction|DoesPlayerHavetheKey (Variables|Default|GetInteractingPlayer)))
    (if _haskey
      (CallFunction|RemoveKey)))
  (Variables|Default|SetPlayerCanInteractWithDoor false)
  (Utilities|FlowControl|DoOnce false
    (:Completed
      (Variables|Default|SetDoorOpened true)
      (CallFunction|DisableCustomKIllz)
      (bind (_direction _dooropen _dooremission) (Timeline|Timeline PlayFromStart)
        (:Update
          (Transformation|SetRelativeTransform (Variables|Default|GetDoorPivot) (Math|Transform|Lerp(Transform) (Math|Transform|MakeTransform "0.000000,0.000000,188.000000") (Math|Transform|MakeTransform "250.000000,0.000000,188.000000" "-110.000000,0.000000,0.000000") _dooropen) false false)
          (Rendering|Material|SetVectorParameterValue (Variables|Default|GetDoorMaterial) "Emissive_Color_Multiplier" (Math|Conversions|ToLinearColor(Float) _dooremission) :!class /Script/Engine.MaterialInstanceDynamic)
          (Rendering|Material|SetVectorParameterValue (Variables|Default|GetFrameMaterial) "Emissive_Color_Multiplier" (Math|Conversions|ToLinearColor(Float) _dooremission) :!class /Script/Engine.MaterialInstanceDynamic))
        (:OpeningSound
          (CallFunction|OpeningSound))
        (:OpeningParticleEffects
          (CallFunction|OpeningParticleFX))
        (:CameraShake
          (CallFunction|CameraShakeLocal))))))

(event Custom|OpenVault
  ; Interact And Open
  (Utilities|FlowControl|DoOnce false
    (:Completed
      (Variables|Default|SetDoorOpened true)
      (CallFunction|DisableCustomKIllz)
      (Timeline|Timeline PlayFromStart))))

(event Custom|ChangeDoorMaterial
  ; Change Door Material
  (Networking|SwitchHasAuthority
    (:Remote
      (Utilities|FlowControl|Sequence
        (:then_0
          (bind _returnvalue (Rendering|Material|CreateDynamicMaterialInstance "/Hermes/Sets/ClayTile/Materials/MI_ClayTile_VaultDoor_A_B.MI_ClayTile_VaultDoor_A_B" "None" "None" :!class /Script/Engine.KismetMaterialLibrary))
          (bind _output_get (Variables|Default|SetDoorMaterial _returnvalue))
          (Rendering|Material|SetMaterial (Variables|Default|GetDoorMesh) 0 _output_get))
        (:then_1
          (bind _returnvalue_2 (Rendering|Material|CreateDynamicMaterialInstance "/Hermes/Sets/ClayTile/Materials/MI_ClayTile_VaultWall_A_B.MI_ClayTile_VaultWall_A_B" "None" "None" :!class /Script/Engine.KismetMaterialLibrary))
          (bind _output_get_2 (Variables|Default|SetFrameMaterial _returnvalue_2))
          (Rendering|Material|SetMaterial (Variables|Default|GetDoorFrame) 0 _output_get_2))))))

(event Custom|OpeningSound
  ; Opening Sound
  (Networking|SwitchHasAuthority
    (:Remote
      (Audio|PlaySoundatLocation (Variables|Default|GetOpen_Sound) (.location (Variables|Default|GetDynamicDoorStartTransform)))
      (Rendering|SetVisibility (Variables|Default|GetDoorKey) true false)
      (Audio|Components|Audio|FadeOut (Variables|Default|GetAudio) 3.0 0.0 "Linear" :!class /Script/Engine.AudioComponent))))

(event Custom|OpeningParticleFX
  ; Opening Particles
  (Networking|SwitchHasAuthority
    (:Remote
      (Effects|Components|ParticleSystem|SpawnFXAtLocation "/Game/Effects/Fort_Effects/Effects/Environment/MagicForest/Systems/NS_VaultDoor_Unlock.NS_VaultDoor_Unlock" (.location (Variables|Default|GetDynamicDoorStartTransform)) (.rotation (Variables|Default|GetDynamicDoorStartTransform)) (.scale (Variables|Default|GetDynamicDoorStartTransform))))))

(event Custom|CameraShakeLocal
  (bind _camerashakebox (Variables|Default|GetCameraShakeBox))
  (bind _overlappingactors (Collision|GetOverlappingActors _camerashakebox "/Script/FortniteGame.FortPlayerPawnAthena" :!class /Script/Engine.PrimitiveComponent))
  ; Local Camera Shakes
  (Networking|SwitchHasAuthority
    (:Remote
      (if (>= (Utilities|Array|Length _overlappingactors) 1)
        (for _array_element _overlappingactors
          (bind _asfort_player_controller_athena (Utilities|Casting|CastToFortPlayerControllerAthena (Fortnite|GetControllerfromActor _array_element)))
          (Game|Feedback|ClientStartCameraShake :self _asfort_player_controller_athena :Shake "/FirePetalDayNightForest/Gameplay/B_VaultOpeningShake.B_VaultOpeningShake_C" :Scale 1.0 :PlaySpace "World"))))))

(event OnComponentBeginOverlap(CustomKillzVolume)
  (params (OverlappedComponent :type Primitive%20Component%20Object%20Reference) (OtherActor :type Actor%20Object%20Reference) (OtherComp :type Primitive%20Component%20Object%20Reference) (OtherBodyIndex :type Integer) (bFromSweep :type Boolean) (SweepResult :type Hit%20Result%20Structure%20%28by%20ref%29))
  ; Prevent Vault Campers who exploit
  (bind _asfort_player_pawn (Utilities|Casting|CastToFortPlayerPawn OtherActor))
  (Networking|SwitchHasAuthority
    (:Authority
      (if (not (Variables|Default|GetDoorOpened))
        (FortPawn|ForceKill :self _asfort_player_pawn :KillerController (Pawn|GetController _asfort_player_pawn) :KillerActor _asfort_player_pawn)))))

(event Custom|DisableCustomKIllz
  ; Prevent Vault Campers who exploit
  (Networking|SwitchHasAuthority
    (:Authority
      (Collision|SetCollisionEnabled (Variables|Default|GetCustomKillzVolume) "NoCollision"))))

(fn BlueprintCanInteract
  (params (InteractingPawn :type Fort%20Pawn%20Object%20Reference) (InteractionBeingAttempted :type EInteractionBeingAttempted%20Enum :default "FirstInteraction") (InteractionType :type TInteractionType%20Enum :default "IT_NoInteraction"))
  (bind _asfort_player_controller_athena (Utilities|Casting|CastToFortPlayerControllerAthena (Pawn|GetController InteractingPawn)))
  (bind _returnvalue (InventoryManagement|GetItemCountforItemDefinition _asfort_player_controller_athena (Variables|Default|GetKeyItemDefinition)))
  (return (and (> _returnvalue 0) (Variables|Default|GetPlayerCanInteractWithDoor))))

(fn BlueprintGetInteractionString
  (params (InteractingPawn :type Fort%20Pawn%20Object%20Reference) (InteractionBeingAttempted :type EInteractionBeingAttempted%20Enum :default "FirstInteraction"))
  (return (Variables|Default|GetInteractionString)))

(fn DoesPlayerHaveTheKey
  (params (Pawn :type Fort%20Pawn%20Object%20Reference))
  (bind _asfort_player_controller_athena (Utilities|Casting|CastToFortPlayerControllerAthena (Pawn|GetController Pawn)))
  (bind _returnvalue (InventoryManagement|GetItemCountforItemDefinition _asfort_player_controller_athena (Variables|Default|GetKeyItemDefinition)))
  (return (> _returnvalue 0)))

(fn OnRep_DoorVisualsChangedState
  (if (Variables|Default|GetDoorVisualsChangedState)
    (Utilities|FlowControl|Sequence
      (:then_0
        (bind _returnvalue (Rendering|Material|CreateDynamicMaterialInstance "/Hermes/Sets/ClayTile/Materials/MI_ClayTile_VaultDoor_A_B.MI_ClayTile_VaultDoor_A_B" "None" "None" :!class /Script/Engine.KismetMaterialLibrary))
        (bind _output_get (Variables|Default|SetDoorMaterial _returnvalue))
        (Rendering|Material|SetMaterial (Variables|Default|GetDoorMesh) 0 _output_get))
      (:then_1
        (bind _returnvalue_2 (Rendering|Material|CreateDynamicMaterialInstance "/Hermes/Sets/ClayTile/Materials/MI_ClayTile_VaultWall_A_B.MI_ClayTile_VaultWall_A_B" "None" "None" :!class /Script/Engine.KismetMaterialLibrary))
        (bind _output_get_2 (Variables|Default|SetFrameMaterial _returnvalue_2))
        (Rendering|Material|SetMaterial (Variables|Default|GetDoorFrame) 0 _output_get_2)))))

(fn RemoveKey
  (bind _asfort_player_controller_athena (Utilities|Casting|CastToFortPlayerControllerAthena (Pawn|GetController (Variables|Default|GetInteractingPlayer))))
  (Inventory|RemoveItemfromInventoryOwner _asfort_player_controller_athena "/FirePetalDayNightForest/Gameplay/KeyItem/WID_FirePetal_DayNight_VaultKey.WID_FirePetal_DayNight_VaultKey"))

(fn ConstructionScript)
