(event Interaction|EventBlueprintOnInteract
  (params (InteractingPawn :type Fort%20Pawn%20Object%20Reference) (InteractionBeingAttempted :type EInteractionBeingAttempted%20Enum :default "FirstInteraction"))
  (bind _interactingplayerref (Variables|Default|GetInteractingPlayerRef))
  ; Interact
  (Variables|Default|SetInteractingPlayerRef InteractingPawn)
  (Variables|Default|SetCanInteractWithNPC false)
  (Utilities|FlowControl|DoOnce false
    (:Completed
      (Utilities|FlowControl|Sequence
        (:then_0
          (Utilities|IsValid _interactingplayerref
            (:Is%20Valid
              (bind _asfort_player_controller_athena (Utilities|Casting|CastToFortPlayerControllerAthena (Pawn|GetController _interactingplayerref)))
              (InventoryManagement|AddItem _asfort_player_controller_athena "/FirePetalDayNightForest/Gameplay/KeyItem/WID_FirePetal_DayNight_VaultKey.WID_FirePetal_DayNight_VaultKey" 1))))
        (:then_1
          (CallFunction|ChangeTODOnClients))
        (:then_2
          (CallFunction|UpdateTempText))))))

(event Custom|ChangeTODOnClients
  ; Time Of Day
  (Networking|SwitchHasAuthority
    (:Remote
      (Class|BFirePetalDayNightVolume|ActivateDaytime (Variables|Default|GetTOD_ActorRef)))))

(event Custom|UpdateTempText
  ; Temp Text
  (Rendering|Components|TextRender|SetText (Variables|Default|GetI%20need%20a%20Sprite!%20) (Utilities|Text|MakeLiteralText "NSLOCTEXT(\"[EF95A48F451DCDC38E18E8CB156C45E8]\", \"C81491C44F370A9EC29423A17B8C9FED\", \"Thanks, here\\'s a Sun Vault Key\")")))

(fn BlueprintCanInteract
  (params (InteractingPawn :type Fort%20Pawn%20Object%20Reference) (InteractionBeingAttempted :type EInteractionBeingAttempted%20Enum :default "FirstInteraction") (InteractionType :type TInteractionType%20Enum :default "IT_NoInteraction"))
  (bind _asfort_player_pawn_athena (Utilities|Casting|CastToFortPlayerPawnAthena InteractingPawn))
  (if (Variables|Default|GetBypassSpriteRequirement)
    (if (Variables|Default|GetCanInteractWithNPC)
      (return true)
      (else
        (return false)))
    (else
      (return (and (Utilities|Array|ContainsItem (GameplayTags|BreakGameplayTagContainer (GameplayTags|GetOwnedGameplayTags _asfort_player_pawn_athena :!class /Script/GameplayTags.BlueprintGameplayTagLibrary)) "(TagName=\"Athena.Player.Sprite.Equipped\")") (Variables|Default|GetCanInteractWithNPC))))))

(fn BlueprintGetInteractionString
  (params (InteractingPawn :type Fort%20Pawn%20Object%20Reference) (InteractionBeingAttempted :type EInteractionBeingAttempted%20Enum :default "FirstInteraction"))
  (return (Variables|Default|GetCubeInteractionString)))

(fn ConstructionScript)
